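# Front-end level sequence for the BP_13 screens that, going by its name, moves the camera from the Choice view to the Choice-Low view.
# Over 15 frames, the Frontend_EventLevel_Camera actor drops and pitches up from looking downward to slightly upward, while its field of view widens.
# Alongside the move, MPC_Cozy's FadePostBattlepass parameter fades from 1 to 0, and two event trigger tracks are present.
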LevelSequence /MiscReferencersOfCosmetics/Environments/FrontEnd/LevelSequence/BP_13/S13_Choice_ToChoice-Low
bindings [
 CameraComponent {
  id=#1
  parent_id=#2
  possessed_class=CameraComponent
  tracks [{
    type=MovieSceneFloatTrack
    display_name=Field Of View
    sections [{type=MovieSceneFloatSection, channels [None {type=MovieSceneScriptingFloatChannel, default=70, keys [[0,70],[15,75]]}]}]
   }]
 }
 Camera_Root_Movement {
  id=#2
  possessed_class=Frontend_EventLevel_Camera_C
  tracks [{
    type=MovieScene3DTransformTrack
    sections [{
      type=MovieScene3DTransformSection
      channels [
       Location.X {type=MovieSceneScriptingDoubleChannel, default=2532.7576, keys [[0,2532.7576],[15,2536.4661]]}
       Location.Y {type=MovieSceneScriptingDoubleChannel, default=-3468.3184, keys [[0,-3468.3184],[15,-3463.9968]]}
       Location.Z {type=MovieSceneScriptingDoubleChannel, default=15076.6445, keys [[0,15076.6445],[15,14954.1377]]}
       Rotation.X {type=MovieSceneScriptingDoubleChannel, default=0.0001, keys [[0,0.0001],[15,0]]}
       Rotation.Y {type=MovieSceneScriptingDoubleChannel, default=-18.2002, keys [[0,-18.2002],[15,3.4772]]}
       Rotation.Z {type=MovieSceneScriptingDoubleChannel, default=-130.5148, keys [[0,-130.5148],[15,-130.5151]]}
       Scale.X {type=MovieSceneScriptingDoubleChannel, default=1}
       Scale.Y {type=MovieSceneScriptingDoubleChannel, default=1}
       Scale.Z {type=MovieSceneScriptingDoubleChannel, default=1}
      ]
     }]
   }]
 }
]
display_rate {denominator=1, numerator=30}
tick_resolution {denominator=1, numerator=24000}
tracks [
 {
  type=MovieSceneEventTrack
  sections [{
    type=MovieSceneEventTriggerSection
    channels [None {type=MovieSceneScriptingEventChannel, const ["<Struct 'MovieSceneEvent' (0x8519ade1ae0) {}>", 1]}]
   }]
 }
 {
  type=MovieSceneEventTrack
  sections [{
    type=MovieSceneEventTriggerSection
    channels [None {type=MovieSceneScriptingEventChannel, const ["<Struct 'MovieSceneEvent' (0x8519ade10e0) {}>", 1]}]
   }]
 }
 {
  type=MovieSceneMaterialParameterCollectionTrack
  display_name=MPC_Cozy
  sections [{type=MovieSceneParameterSection, channels [FadePostBattlepass {type=MovieSceneScriptingFloatChannel, default=1, keys [[0,1],[15,0]]}]}]
 }
]
