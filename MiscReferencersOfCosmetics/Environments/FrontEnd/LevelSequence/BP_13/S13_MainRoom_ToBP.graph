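# Front-end level sequence that, judging by its name, moves the season 13 lobby camera from the main room to the battle pass view.
# It animates the Frontend_EventLevel_Camera transform along a swooping path with slight rotation changes while its CameraComponent field of view widens.
# An MPC_Cozy material parameter track drops FadePostBattlepass from one to zero after the camera move ends.
# Two event trigger tracks fire sequence events.
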LevelSequence /MiscReferencersOfCosmetics/Environments/FrontEnd/LevelSequence/BP_13/S13_MainRoom_ToBP
bindings [
 CameraComponent {
  id=#1
  parent_id=#2
  possessed_class=CameraComponent
  tracks [{
    type=MovieSceneFloatTrack
    display_name=Field Of View
    sections [{type=MovieSceneFloatSection, channels [None {type=MovieSceneScriptingFloatChannel, default=55, keys [[0,55],[50,70]]}]}]
   }]
 }
 Camera_Root_Movement {
  id=#2
  possessed_class=Frontend_EventLevel_Camera_C
  tracks [{
    type=MovieScene3DTransformTrack
    sections [{
      type=MovieScene3DTransformSection
      channels [
       Location.X {type=MovieSceneScriptingDoubleChannel, default=296.733, keys [[0,4057.3916],[24,2145.3977],[50,296.733]]}
       Location.Y {type=MovieSceneScriptingDoubleChannel, default=-2572.9426, keys [[0,-2254.5737],[24,-2448.658],[50,-2572.9426]]}
       Location.Z {type=MovieSceneScriptingDoubleChannel, default=14387.5117, keys [[0,14608.8975],[24,14748.8223],[40,14430.2217],[50,14387.5117]]}
       Rotation.X {type=MovieSceneScriptingDoubleChannel, default=-0.0543, keys [[0,0],[24,-2],[50,-0.0543]]}
       Rotation.Y {type=MovieSceneScriptingDoubleChannel, default=-0.1871, keys [[0,0.045],[50,-0.1871]]}
       Rotation.Z {type=MovieSceneScriptingDoubleChannel, default=-159.1766, keys [[0,-163.0003],[50,-159.1766]]}
       Scale.X {type=MovieSceneScriptingDoubleChannel, default=1}
       Scale.Y {type=MovieSceneScriptingDoubleChannel, default=1}
       Scale.Z {type=MovieSceneScriptingDoubleChannel, default=1}
      ]
     }]
   }]
 }
]
display_rate {denominator=1, numerator=30}
tick_resolution {denominator=1, numerator=24000}
tracks [
 {
  type=MovieSceneEventTrack
  sections [{
    type=MovieSceneEventTriggerSection
    channels [None {type=MovieSceneScriptingEventChannel, const ["<Struct 'MovieSceneEvent' (0x8519ade2bc0) {}>", 1]}]
   }]
 }
 {
  type=MovieSceneEventTrack
  sections [{
    type=MovieSceneEventTriggerSection
    channels [None {type=MovieSceneScriptingEventChannel, const ["<Struct 'MovieSceneEvent' (0x8519ade2b20) {}>", 1]}]
   }]
 }
 {
  type=MovieSceneMaterialParameterCollectionTrack
  display_name=MPC_Cozy
  sections [{
    type=MovieSceneParameterSection
    channels [FadePostBattlepass {type=MovieSceneScriptingFloatChannel, default=1, thinned_from=4, keys [[0,1],[64,1],[79.9987,0]]}]
   }]
 }
]
