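# Frontend Level Sequence, named as the spy-base main room to Ocean Rider inspect transition, that moves the camera while powering up a hologram table.
# The event-level camera root follows a keyed move and turn as its field of view widens over about one second.
# Meanwhile material parameter tracks animate the holo rings, point cloud, lights and cylinder top, alongside transform tracks and a Niagara map position offset.
# Two root event tracks fire triggers.
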LevelSequence /MiscReferencersOfCosmetics/Environments/FrontEnd/LevelSequence/BP_13/S13_MainRoom_ToOceanRider_Inspect
bindings [
 sm_spybase_table_holo_ring_vertical_squares_fx {id=#1, possessed_class=FortStaticMeshActor, tracks []}
 sm_cylinder_top {id=#2, possessed_class=FortStaticMeshActor, tracks []}
 StaticMeshComponent0 {
  id=#3
  parent_id=#1
  possessed_class=StaticMeshComponent
  tracks [{
    type=MovieSceneComponentMaterialTrack
    display_name=Material Element 0
    sections [{
      type=MovieSceneParameterSection
      channels [
       Speed Lerp {type=MovieSceneScriptingFloatChannel, default=0, keys [[2,1],[16,0.8],[20,0]]}
       Movement Multiplier {type=MovieSceneScriptingFloatChannel, default=15, keys [[0,15],[2,15],[9,12],[20,18]]}
       Distance Offset {type=MovieSceneScriptingFloatChannel, default=550, keys [[0,550],[20,225]]}
       ColorR {type=MovieSceneScriptingFloatChannel, default=0.0056, const [0.0056,1]}
       ColorG {type=MovieSceneScriptingFloatChannel, default=0.0452, const [0.0452,1]}
       ColorB {type=MovieSceneScriptingFloatChannel, default=0.0833, const [0.0833,1]}
       ColorA {type=MovieSceneScriptingFloatChannel, default=1, const [1,1]}
      ]
     }]
   }]
 }
 StaticMeshComponent0 {
  id=#4
  parent_id=#5
  possessed_class=StaticMeshComponent
  tracks [{
    type=MovieSceneComponentMaterialTrack
    display_name=Material Element 0
    sections [{
      type=MovieSceneParameterSection
      channels [
       SpriteScale {type=MovieSceneScriptingFloatChannel, default=3, keys [[0,0],[14,3],[20,1.25]]}
       Distance Offset {type=MovieSceneScriptingFloatChannel, default=325}
       Distance Divisor {type=MovieSceneScriptingFloatChannel, default=256, const [256,1]}
       ColorR {type=MovieSceneScriptingFloatChannel, default=0, const [0.0671,1]}
       ColorG {type=MovieSceneScriptingFloatChannel, default=0.54, const [0.5424,1]}
       ColorB {type=MovieSceneScriptingFloatChannel, default=1, const [1,1]}
       ColorA {type=MovieSceneScriptingFloatChannel, default=1, keys [[2,1],[20,0.5],[20.0013,1]]}
      ]
     }]
   }]
 }
 CameraComponent {
  id=#6
  parent_id=#7
  possessed_class=CameraComponent
  tracks [{
    type=MovieSceneFloatTrack
    display_name=Field Of View
    sections [{type=MovieSceneFloatSection, channels [None {type=MovieSceneScriptingFloatChannel, default=80, keys [[0,55],[30,78]]}]}]
   }]
 }
 sm_spybase_table_holo_ring_fx {id=#8, possessed_class=FortStaticMeshActor, tracks []}
 StaticMeshComponent0 {
  id=#9
  parent_id=#8
  possessed_class=StaticMeshComponent
  tracks [{
    type=MovieSceneComponentMaterialTrack
    display_name=Material Element 0
    sections [{
      type=MovieSceneParameterSection
      channels [Blocks Intensity {type=MovieSceneScriptingFloatChannel, default=1.5, keys [[2,2],[14,1.5],[20.0013,0.4]]}]
     }]
   }]
 }
 SM_Cylinder_Floor_Pad_Glow {
  id=#10
  possessed_class=FortStaticMeshActor
  tracks [{
    type=MovieScene3DTransformTrack
    sections [{
      type=MovieScene3DTransformSection
      channels [
       Location.X {type=MovieSceneScriptingDoubleChannel, default=2048, const [2048,1]}
       Location.Y {type=MovieSceneScriptingDoubleChannel, default=-2560, const [-2560,1]}
       Location.Z {type=MovieSceneScriptingDoubleChannel, default=99.1319, const [99.1319,1]}
       Rotation.X {type=MovieSceneScriptingDoubleChannel, default=0, const [0,1]}
       Rotation.Y {type=MovieSceneScriptingDoubleChannel, default=0, const [0,1]}
       Rotation.Z {type=MovieSceneScriptingDoubleChannel, default=0, const [0,1]}
       Scale.X {type=MovieSceneScriptingDoubleChannel, default=2.185, const [2.185,1]}
       Scale.Y {type=MovieSceneScriptingDoubleChannel, default=2.185, const [2.185,1]}
       Scale.Z {type=MovieSceneScriptingDoubleChannel, default=2, keys [[0,0.5],[9,0.5],[16,2],[20,1]]}
      ]
     }]
   }]
 }
 sm_spybase_table_holo_ring_point_cloud_02_fx {id=#5, possessed_class=FortStaticMeshActor, tracks []}
 Camera_Root_Movement {
  id=#7
  possessed_class=Frontend_EventLevel_Camera_C
  tracks [{
    type=MovieScene3DTransformTrack
    sections [{
      type=MovieScene3DTransformSection
      channels [
       Location.X {
        type=MovieSceneScriptingDoubleChannel
        default=3134.053
        keys [[0,4057.3916],[6,4049.2046],[12,3884.2986],[20,3371.4678],[25,3194.1438],[30,3134.053]]
       }
       Location.Y {
        type=MovieSceneScriptingDoubleChannel
        default=-1850.3755
        keys [[0,-2254.5737],[6,-2249.8564],[14,-2176.1394],[20,-1967.6426],[25,-1877.4556],[30,-1850.3755]]
       }
       Location.Z {type=MovieSceneScriptingDoubleChannel, default=14580.2656, thinned_from=6, keys [[0,14608.8975],[30,14580.2656]]}
       Rotation.X {type=MovieSceneScriptingDoubleChannel, default=0, keys [[0,0],[14,-3.4232],[25,-0.1589]]}
       Rotation.Y {type=MovieSceneScriptingDoubleChannel, default=0.045, keys [[0,0.045],[6,0.0291],[14,-0.3625],[20,-1.3119],[25,-1.8676],[30,-2]]}
       Rotation.Z {
        type=MovieSceneScriptingDoubleChannel
        default=-221.27
        keys [[0,-163.0003],[6,-163.4526],[14,-174.6105],[20,-205.6792],[25,-217.9304],[30,-221.27]]
       }
      ]
     }]
   }]
 }
 StaticMeshComponent0 {
  id=#11
  parent_id=#12
  possessed_class=StaticMeshComponent
  tracks [{
    type=MovieSceneComponentMaterialTrack
    display_name=Material Element 0
    sections [{
      type=MovieSceneParameterSection
      channels [
       Distance Divisor {type=MovieSceneScriptingFloatChannel, default=256, const [256,2]}
       Distance Offset {type=MovieSceneScriptingFloatChannel, default=365, keys [[0,665.2485],[20,365]]}
       Alpha {type=MovieSceneScriptingFloatChannel, default=1, keys [[0,0.2],[20,1]]}
      ]
     }]
   }]
 }
 NiagaraComponent0 {
  id=#13
  parent_id=#14
  possessed_class=NiagaraComponent
  tracks [{
    type=MovieSceneNiagaraVectorParameterTrack
    display_name=PositionOffset
    sections [{
      type=MovieSceneFloatVectorSection
      channels [
       Vector.X {type=MovieSceneScriptingFloatChannel}
       Vector.Y {type=MovieSceneScriptingFloatChannel}
       Vector.Z {type=MovieSceneScriptingFloatChannel, default=200, keys [[0,140],[20,200]]}
      ]
     }]
   }]
 }
 NS_Battle_Bus_Spybase_Map {id=#14, possessed_class=NiagaraActor, tracks []}
 StaticMeshComponent0 {
  id=#15
  parent_id=#2
  possessed_class=StaticMeshComponent
  tracks [
   {
    type=MovieSceneComponentMaterialTrack
    display_name=Material Element 0
    sections [{
      type=MovieSceneParameterSection
      channels [
       Alpha {type=MovieSceneScriptingFloatChannel, default=2, keys [[3,0],[12,2],[19,1]]}
       ColorR {type=MovieSceneScriptingFloatChannel, default=0.25, keys [[12,0.1],[20,0.25]]}
       ColorG {type=MovieSceneScriptingFloatChannel, default=0.55, keys [[12,0.4],[20,0.55]]}
       ColorB {type=MovieSceneScriptingFloatChannel, const [1,1]}
       ColorA {type=MovieSceneScriptingFloatChannel, const [1,1]}
      ]
     }]
   }
   {
    type=MovieScene3DTransformTrack
    sections [{
      type=MovieScene3DTransformSection
      channels [
       Location.X {type=MovieSceneScriptingDoubleChannel, default=2048, const [2048,1]}
       Location.Y {type=MovieSceneScriptingDoubleChannel, default=-2560, const [-2560,1]}
       Location.Z {type=MovieSceneScriptingDoubleChannel, default=127.6264, const [127.6264,1]}
       Rotation.X {type=MovieSceneScriptingDoubleChannel, default=0}
       Rotation.Y {type=MovieSceneScriptingDoubleChannel, default=0}
       Rotation.Z {type=MovieSceneScriptingDoubleChannel, default=0, keys [[0,0],[20,150]]}
       Scale.X {type=MovieSceneScriptingDoubleChannel, default=2.28, const [2.28,1]}
       Scale.Y {type=MovieSceneScriptingDoubleChannel, default=2.28, const [2.28,1]}
       Scale.Z {type=MovieSceneScriptingDoubleChannel, default=2.28, const [2.28,1]}
      ]
     }]
   }
  ]
 }
 sm_spybase_table_holo_ring_lights_meshsprites_fx {
  id=#12
  possessed_class=FortStaticMeshActor
  tracks [{
    type=MovieScene3DTransformTrack
    sections [{
      type=MovieScene3DTransformSection
      channels [
       Location.X {type=MovieSceneScriptingDoubleChannel, default=2048, const [2048,1]}
       Location.Y {type=MovieSceneScriptingDoubleChannel, default=-2560, const [-2560,1]}
       Location.Z {type=MovieSceneScriptingDoubleChannel, default=104.1381, const [104.1381,1]}
       Rotation.X {type=MovieSceneScriptingDoubleChannel, default=0, const [0,1]}
       Rotation.Y {type=MovieSceneScriptingDoubleChannel, default=0, const [0,1]}
       Rotation.Z {type=MovieSceneScriptingDoubleChannel, default=0, const [0,1]}
       Scale.X {type=MovieSceneScriptingDoubleChannel, default=1, const [1,1]}
       Scale.Y {type=MovieSceneScriptingDoubleChannel, default=1, const [1,1]}
       Scale.Z {type=MovieSceneScriptingDoubleChannel, default=1, const [1,1]}
      ]
     }]
   }]
 }
]
display_rate {denominator=1, numerator=30}
tick_resolution {denominator=1, numerator=24000}
tracks [
 {
  type=MovieSceneEventTrack
  sections [{
    type=MovieSceneEventTriggerSection
    channels [None {type=MovieSceneScriptingEventChannel, const ["<Struct 'MovieSceneEvent' (0x8519ade2080) {}>", 1]}]
   }]
 }
 {
  type=MovieSceneEventTrack
  sections [{
    type=MovieSceneEventTriggerSection
    channels [None {type=MovieSceneScriptingEventChannel, const ["<Struct 'MovieSceneEvent' (0x8519ade24e0) {}>", 1]}]
   }]
 }
]
