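# Front-end camera transition for the BP_13 scene, moving from the main room to the umbrella view.
# Keyframes the Frontend_EventLevel_Camera actor's location and rotation into a pitched-down, re-yawed framing, while its child CameraComponent widens field of view.
# Over the same key range, an MPC_Cozy track fades FadePostBattlepass from one to zero, driving materials in step with the move.
# Two event trigger tracks each fire an event whose target is not shown.
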LevelSequence /MiscReferencersOfCosmetics/Environments/FrontEnd/LevelSequence/BP_13/S13_MainRoom_ToUmbrella
bindings [
 CameraComponent {
  id=#1
  parent_id=#2
  possessed_class=CameraComponent
  tracks [{
    type=MovieSceneFloatTrack
    display_name=Field Of View
    sections [{type=MovieSceneFloatSection, channels [None {type=MovieSceneScriptingFloatChannel, default=70, keys [[0,55],[50,70]]}]}]
   }]
 }
 Camera_Root_Movement {
  id=#2
  possessed_class=Frontend_EventLevel_Camera_C
  tracks [{
    type=MovieScene3DTransformTrack
    sections [{
      type=MovieScene3DTransformSection
      channels [
       Location.X {type=MovieSceneScriptingDoubleChannel, default=2519.8735, keys [[0,4057.3916],[50,2519.8735]]}
       Location.Y {type=MovieSceneScriptingDoubleChannel, default=-3331.3691, keys [[0,-2254.5574],[50,-3331.3691]]}
       Location.Z {type=MovieSceneScriptingDoubleChannel, default=15059.9551, keys [[0,14608.8975],[50,15059.9551]]}
       Rotation.X {type=MovieSceneScriptingDoubleChannel, default=0, keys [[0,0],[50,0]]}
       Rotation.Y {type=MovieSceneScriptingDoubleChannel, default=-15, keys [[0,0.045],[50,-15]]}
       Rotation.Z {type=MovieSceneScriptingDoubleChannel, default=-125.3078, keys [[0,-163.0003],[50,-125.3078]]}
       Scale.X {type=MovieSceneScriptingDoubleChannel, default=1}
       Scale.Y {type=MovieSceneScriptingDoubleChannel, default=1}
       Scale.Z {type=MovieSceneScriptingDoubleChannel, default=1}
      ]
     }]
   }]
 }
]
display_rate {denominator=1, numerator=30}
tick_resolution {denominator=1, numerator=24000}
tracks [
 {
  type=MovieSceneEventTrack
  sections [{
    type=MovieSceneEventTriggerSection
    channels [None {type=MovieSceneScriptingEventChannel, const ["<Struct 'MovieSceneEvent' (0x8519ade4ba0) {}>", 1]}]
   }]
 }
 {
  type=MovieSceneEventTrack
  sections [{
    type=MovieSceneEventTriggerSection
    channels [None {type=MovieSceneScriptingEventChannel, const ["<Struct 'MovieSceneEvent' (0x8519ade30c0) {}>", 1]}]
   }]
 }
 {
  type=MovieSceneMaterialParameterCollectionTrack
  display_name=MPC_Cozy
  sections [{type=MovieSceneParameterSection, channels [FadePostBattlepass {type=MovieSceneScriptingFloatChannel, default=1, keys [[0,1],[50,0]]}]}]
 }
]
